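# Physics asset for the Durrburger Hero glider skeletal mesh, giving its food-themed parts collision bodies and joints.
# Each body is a single capsule on a bone, covering boxes, tubes, spatula wings, pans, shakers, fries and handles.
# Constraints radiate from lower_box, with both fries hanging off pan_l and handle chains branching from handles, the right having an extra 01 link.
# Every joint shares the same 45-degree swing and twist limit values.
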PhysicsAsset /BRCosmetics/Gadgets/Assets/VinderTech_GliderChute/Glider_Durrburger_Hero/Meshes/SK_Glider_Durrburger_Hero_Physics
bodies {
 b_wing_l_spatchula [{shapeType=Capsule, length=90.4709, radius=17.23, center=<83.296 -1.1579 -9.4236>, rotation=<rot 0 -88.877 -80.1076>}]
 b_wing_r_spatchula [{shapeType=Capsule, length=90.7663, radius=17.6, center=<83.5268 -1.5252 -10.2421>, rotation=<rot 0 -89.0906 -79.2399>}]
 dyn_fry_01 [{shapeType=Capsule, length=25.7797, radius=4.2868, center=<-2.89 0.8104 0.8077>, rotation=<rot 0 -106.1109 -89.2543>}]
 dyn_fry_02 [{shapeType=Capsule, length=24.0758, radius=3.5667, center=<0.2381 -0.3104 0.3316>, rotation=<rot 0 35.8873 -89.9853>}]
 f_wing_l_spatchula [{shapeType=Capsule, length=77.5345, radius=16.7107, center=<70.9915 -1.5304 0.4537>, rotation=<rot 0 92.583 -88.0724>}]
 f_wing_r_spatchula [{shapeType=Capsule, length=77.2488, radius=16.7874, center=<70.56 -1.6251 0.5034>, rotation=<rot 0 89.9103 -87.9831>}]
 handles [{shapeType=Capsule, length=48.6544, radius=12.3153, center=<9.0975 -7.2061 -6.8548>, rotation=<rot 0 -86.8448 -81.5515>}]
 l_handle_05 [{shapeType=Capsule, length=27.6629, radius=5.1198, center=<-16.6739 0.6932 -1.0913>, rotation=<rot 0 87.3675 -84.4634>}]
 l_handle_08 [{shapeType=Capsule, length=19.8535, radius=13.7433, center=<-11.7552 -10.4686 -1.0622>, rotation=<rot 0 156.325 -81.6395>}]
 lower_box [{shapeType=Capsule, length=72.6177, radius=52.5032, center=<1.3453 50.5404 8.7935>, rotation=<rot 0 89.0241 -89.9322>}]
 pan_l [{shapeType=Capsule, length=95.8187, radius=65.0504, center=<-1.1677 -4.9007 -1.5893>, rotation=<rot 0 96.1878 -89.368>}]
 pan_r [{shapeType=Capsule, length=96.132, radius=66.9488, center=<2.5985 -0.9977 0.0241>, rotation=<rot 0 94.8266 -89.6995>}]
 r_handle_01 [{shapeType=Capsule, length=15.0209, radius=5.7679, center=<10.8487 -5.9696 -3.2156>, rotation=<rot 0 -140.7432 -67.4775>}]
 r_handle_05 [{shapeType=Capsule, length=22.5696, radius=4.9042, center=<21.6752 -0.7609 1.368>, rotation=<rot 0 86.4642 -83.1769>}]
 r_handle_08 [{shapeType=Capsule, length=18.7016, radius=11.0493, center=<14.4658 10.5722 0.4856>, rotation=<rot 0 157.3359 -81.3484>}]
 shaker_l [{shapeType=Capsule, length=33.816, radius=20.2457, center=<-0.1295 0.4051 0.292>, rotation=<rot 0 -178.9901 -87.4369>}]
 shaker_r [{shapeType=Capsule, length=33.815, radius=20.2481, center=<-0.1473 -0.4054 0.2801>, rotation=<rot 0 -1.3287 -87.5925>}]
 tube_01 [{shapeType=Capsule, length=86.9503, radius=27.3674, center=<0.2579 16.7512 0.2352>, rotation=<rot 0 179.549 -89.3955>}]
 tube_02 [{shapeType=Capsule, length=86.4857, radius=27.5205, center=<0.5391 16.2899 0.2718>, rotation=<rot 0 -179.963 -89.2497>}]
 upper_box [{shapeType=Capsule, length=57.2458, radius=54.6712, center=<0.8724 42.4684 15.2633>, rotation=<rot 4.6623 104.7449 -90>}]
}
constraints [
 {bone1Name=upper_box, bone2Name=lower_box, swing1LimitDegrees=45, swing2LimitDegrees=45, twistLimitDegrees=45}
 {bone1Name=tube_01, bone2Name=lower_box, swing1LimitDegrees=45, swing2LimitDegrees=45, twistLimitDegrees=45}
 {bone1Name=tube_02, bone2Name=lower_box, swing1LimitDegrees=45, swing2LimitDegrees=45, twistLimitDegrees=45}
 {bone1Name=f_wing_l_spatchula, bone2Name=lower_box, swing1LimitDegrees=45, swing2LimitDegrees=45, twistLimitDegrees=45}
 {bone1Name=b_wing_l_spatchula, bone2Name=lower_box, swing1LimitDegrees=45, swing2LimitDegrees=45, twistLimitDegrees=45}
 {bone1Name=shaker_l, bone2Name=lower_box, swing1LimitDegrees=45, swing2LimitDegrees=45, twistLimitDegrees=45}
 {bone1Name=pan_l, bone2Name=lower_box, swing1LimitDegrees=45, swing2LimitDegrees=45, twistLimitDegrees=45}
 {bone1Name=dyn_fry_01, bone2Name=pan_l, swing1LimitDegrees=45, swing2LimitDegrees=45, twistLimitDegrees=45}
 {bone1Name=dyn_fry_02, bone2Name=pan_l, swing1LimitDegrees=45, swing2LimitDegrees=45, twistLimitDegrees=45}
 {bone1Name=f_wing_r_spatchula, bone2Name=lower_box, swing1LimitDegrees=45, swing2LimitDegrees=45, twistLimitDegrees=45}
 {bone1Name=b_wing_r_spatchula, bone2Name=lower_box, swing1LimitDegrees=45, swing2LimitDegrees=45, twistLimitDegrees=45}
 {bone1Name=shaker_r, bone2Name=lower_box, swing1LimitDegrees=45, swing2LimitDegrees=45, twistLimitDegrees=45}
 {bone1Name=pan_r, bone2Name=lower_box, swing1LimitDegrees=45, swing2LimitDegrees=45, twistLimitDegrees=45}
 {bone1Name=handles, bone2Name=lower_box, swing1LimitDegrees=45, swing2LimitDegrees=45, twistLimitDegrees=45}
 {bone1Name=r_handle_01, bone2Name=handles, swing1LimitDegrees=45, swing2LimitDegrees=45, twistLimitDegrees=45}
 {bone1Name=r_handle_05, bone2Name=r_handle_01, swing1LimitDegrees=45, swing2LimitDegrees=45, twistLimitDegrees=45}
 {bone1Name=r_handle_08, bone2Name=r_handle_05, swing1LimitDegrees=45, swing2LimitDegrees=45, twistLimitDegrees=45}
 {bone1Name=l_handle_05, bone2Name=handles, swing1LimitDegrees=45, swing2LimitDegrees=45, twistLimitDegrees=45}
 {bone1Name=l_handle_08, bone2Name=l_handle_05, swing1LimitDegrees=45, swing2LimitDegrees=45, twistLimitDegrees=45}
]
